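# Additive, unlit, two-sided beam-trail material for the air-current trails of the BlackMondayCape glider.
# Emissive multiplies a fast-panning distortion texture on swizzled UV1 by a high-to-mid color gradient along U and particle color.
# Opacity combines particle alpha, an asymmetric fade along U and a sine falloff across V.
# Index-of-refraction distortion uses panning noise and SepTrans dynamic parameter, flattened to 1 on mobile and low/medium quality.
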Material /Game/Gadgets/Assets/VinderTech_GliderChute/Glider_BlackMondayCape/FX/M_Glider_BlackMondayCake_AirCurTrail
graph {
 properties {
  bUsedWithBeamTrails=true
  blendMode=BLEND_Additive
  pixelDepthOffsetMode=PDOM_Legacy
  refractionMethod=RM_IndexOfRefraction
  shadingModel=MSM_Unlit
  twoSided=true
  usedShadingModels=Unlit
 }
 nodes [
  n1 = Multiply(
   A=MaterialFunctionCall(XY=TextureCoordinate() {coordinateIndex=1}) {materialFunction=&Swizzle}.YX
   B=AppendVector(
    A=ScalarParameter() {defaultValue=1, group=EmisColor, parameterName=U_Tile}
    B=ScalarParameter() {defaultValue=0.25, group=EmisColor, parameterName=V_Tile}
   )
  )
  n2 = LinearInterpolate(Alpha=TextureSample(UVs=Panner(Coordinate=@n1) {speedX=0.1, speedY=-8}) {texture=/Game/Global/Textures/Default/Masks_and_Gradients/T_BlasterBeamDistort}.R) {constA=0.3, constB=3}
  n3 = ComponentMask(None=TextureCoordinate()) {r=true}
  n4 = ScalarParameter() {defaultValue=3, parameterName=ColorLerpExp}
  n5 = LinearInterpolate(
   A=VectorParameter() {parameterName=ColorHigh, defaultValue=<rgba 0.9 0.9 1 0>}.RGB
   B=VectorParameter() {parameterName=ColorMid, defaultValue=<rgba 0.0281 0.0281 0.0312 0>}.RGB
   Alpha=Saturate(None=Power(Base=@n3, Exp=@n4) {constExponent=3})
  )
  n6 = ParticleColor()
  MaterialFunctionCall(Emissive=Multiply(A=Multiply(A=@n2, B=@n5) {constB=10}, B=@n6.RGB)) {materialFunction=&MF_FX_EmissiveAdjustments}
  n7 = TextureCoordinate()
  n8 = OneMinus(None=ComponentMask(None=TextureCoordinate()) {r=true})
  n9 = Constant() {r=4}
  n10 = Multiply(A=Multiply(A=ComponentMask(None=@n7) {r=true}, B=Power(Base=@n8, Exp=@n9))) {constB=10}
  n11 = ComponentMask(None=TextureCoordinate()) {g=true}
  Multiply(A=Multiply(A=@n6.A, B=Saturate(None=@n10)), B=Power(Base=Sine(None=@n11) {period=2}))
  n12 = LinearInterpolate(
   B=LinearInterpolate(Alpha=DynamicParameter() {defaultValue=<rgba 0 1 1 1>, paramNames [SepTrans,Param2,Param3,Param4]}.SepTrans) {constA=1.1, constB=1.035}
   Alpha=ScalarParameter() {defaultValue=15, parameterName=RefIntensity}
  ) {constA=1, constB=1.1}
  n13 = TextureCoordinate()
  n14 = Multiply(
   A=MaterialFunctionCall(XY=TextureCoordinate() {coordinateIndex=1}) {materialFunction=&Swizzle}.YX
   B=AppendVector(
    A=ScalarParameter() {defaultValue=1, group=Normal, parameterName=U_Tile_N}
    B=ScalarParameter() {defaultValue=0.25, group=Normal, parameterName=V_Tile_N}
   )
  )
  n15 = Multiply(
   A=Power(Base=ComponentMask(None=@n13) {r=true}, Exp=ScalarParameter() {defaultValue=2, parameterName=RefFadeOut})
   B=TextureSampleParameter2D(UVs=Panner(Coordinate=@n14) {speedY=-10}) {parameterName=RefTexture, texture=/Game/Effects/Fort_Effects/Textures/Tile/T_TileNoise_03}.R
  )
  n16 = Constant() {r=1}
  QualitySwitch(
   Default=ShadingPathSwitch(Default=LinearInterpolate(B=@n12, Alpha=@n15) {constA=1, constB=1.1}, Mobile=@n16) {desc=Migrated from FeatureLevelSwitch}
   Low=@n16
   Medium=@n16
  )
 ]
}
